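# Niagara cosmetic effect for the PiteousKicks Pine character skin, adding green glows, smoke and trails to eyes, hands, arms, legs and feet.
# Sprites and meshes follow skeletal bones, while ribbons wrap limbs through FN_BoneFlow.
# Mesh trails store smoothed, lagging positions in Interpolate Dynamic.
# User parameters drive alpha, dissolve, body-part hiding and staged kills by TransformProgress, with brightness adjusted for front end, time of day and platform.
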NiagaraSystem /BRCosmetics/Characters/Player/Male/Medium/Bodies/M_MED_PiteousKicks_Pine/Fx/NS_PiteousKicks_Pine_Skin_Trail
dependencies {
 usedDataInterfaces [
  &NiagaraDataInterfaceArrayFloat
  &NiagaraDataInterfaceCurve
  &NiagaraDataInterfaceColorCurve
  &NiagaraDataInterfacePlatformSet
  &NiagaraDataInterfaceArrayFloat3
  &NiagaraDataInterfaceArrayInt32
  &NiagaraDataInterfaceSpriteRendererInfo
 ]
 usedDynamicInputs []
 usedModules [&SystemState]
}
emitters {
 FN_Debug {
  bEnabled=false
  emitterUpdateScript [
   &EmitterState()
   &SpawnBurst_Instantaneous(Spawn Count=1, Use Spawn Probability=false)
  ]
  particleSpawnScript [/Niagara/Modules/Spawn/Initialization/InitializeParticle(
    Lifetime=5
    Module.Write Color=true
    Module.Write Lifetime=true
    Module.Write Mass=true
    Module.Write Material Random=true
    Module.Write Position=false
    Module.Write Scale=false
    Module.Write SpriteRotation=false
    Module.Write SpriteSize=true
    Position=&SimulationPosition()
    Sprite Size=<100 100>
   )]
  particleUpdateScript [
   &ParticleState()
   /SetVariables(Particles.Debug=0.1) {isSetParameters=true}
   &DynamicMaterialParameters(
    Index 0 Param 1=&MakeCustomFloatFromBool(Bool=$User.bisFootHidden, True Float=1, False Float=0)
    Module.Param0WriteEnabled=true
    Module.Param1WriteEnabled=true
    Module.Param2WriteEnabled=true
    Module.Param3WriteEnabled=true
   )
   &SystemLocation(MaskSpawnGroup=false, Offset=<0 0 200>)
  ]
  rendererClasses [&NiagaraSpriteRendererProperties]
  renderers [{rendererClass=&NiagaraSpriteRendererProperties}]
 }
 Hand_Smoke {
  emitterUpdateScript [
   &EmitterState(Inactive Response=::NewEnumerator1, Life Cycle Mode=::NewEnumerator1, Loop Duration=1, UseLoopDelay=false)
   &SpawnRate(SpawnRate=20, Use Spawn Probability=false)
   &SpawnBurst_Instantaneous(Spawn Count=2, Use Spawn Probability=false, UseLoopCountLimit=false)
  ]
  particleSpawnScript [
   &FN_SetSpawnGroup(Number of Spawn Groups=2)
   /Niagara/Modules/Spawn/Initialization/V2/InitializeParticle(
    Color=&MakeCustomLinearColorFromBool(Bool=&RandomBool(), True Color=<rgba 0.1919 1 0.0885 1.2>, False Color=<rgba 0.0781 1 0.1608 1.2>)
    Lifetime Max=1.2
    Lifetime Min=0.8
    Lifetime Mode=::NewEnumerator1
    Sprite Rotation Mode=::NewEnumerator2
    Sprite Size Mode=::NewEnumerator1
    Uniform Sprite Size Max=16
    Uniform Sprite Size Min=13
    UsePositionOffset=false
    Write Material Random=true
   )
   &SubUVAnimation(
    End Frame Range Override=4
    "Sprite Renderer "=DataInterface() {propertyValues {SpriteRenderer=&self:Hand_Smoke_0.NiagaraSpriteRendererProperties_1}}
    SubUV Animation Mode:ENiagara_SubUVLookupModeV2=::NewEnumerator2
    UseEndFrame=true
    UseStartFrame=true
   )
   &CameraOffset(Camera Offset Amount=&RandomRangeFloat(Minimum=0, Maximum=20))
   &AddVelocity(Velocity=<0 0 1>)
  ]
  particleUpdateScript [
   &ParticleState()
   &SkeletalMeshLocation(
    Bone / Socket Index=$Particles.SpawnGroup
    Bone Sampling Mode=::NewEnumerator4
    MaskSpawnGroup=false
    Sampled Position Offset=&SelectVectorFromArray(Direct Array Index=$Particles.SpawnGroup, Vector Selection Array=DataInterface() {propertyValues {FloatData [<10 -2 0>, <-10 2 0>]}})
    Skeletal Mesh=DataInterface() {propertyValues {
      DefaultMesh=null
      ExcludeBoneName=root
      OptionalComponentClass=null
      PreviewMesh=/Game/Characters/Player/Male/Medium/Bodies/M_MED_PiteousKicks/Meshes/M_MED_PiteousKicks
      SkinningMode=SkinOnTheFly
      SoftSourceActor=null
      SourceMode=Default
      UvSetIndex=0
      WholeMeshLOD=-1
      bExcludeBone=false
      bReadDeformedGeometry=true
      bRequireCurrentFrameData=true
      ComponentTags []
      FilteredBones [hand_l,hand_r]
      FilteredSockets []
      MeshUserParameter {parameter {name=None, typeDefHandle {registeredTypeIndex=1}, varData []}}
      SamplingRegions []
     }}
   )
   &ScaleSpriteSize(Scale Factor=&Multiply_Vector2(
     A=<1 1>
     B=&Vector2DFromFloat(Value=&FloatFromCurve(FloatCurve=DataInterface() {propertyValues {Curve {keys [<at 0 1>,<at 1 0.8>]}}}))
    ))
   &SpriteRotationRate(Rotation Rate=&FloatFromCurve(
     FloatCurve=DataInterface() {propertyValues {Curve {keys [{interpMode=RCIM_Cubic, tangentMode=RCTM_Break, time=0, value=1}, {interpMode=RCIM_Cubic, tangentMode=RCTM_Break, time=1, value=0.489}]}}}
     Scale Curve=&Lerp_Float(A=-80, B=80, Alpha=$Particles.Initial.MaterialRandom)
    ))
   &ScaleColor(
    Scale Alpha=&FloatFromCurve(FloatCurve=DataInterface() {propertyValues {Curve {keys [<at 0 0 RCIM_Cubic>,<at 0.4 1 RCIM_Cubic>,<at 1 0 RCIM_Cubic>]}}}, Scale Curve=0.3)
    ScaleA=true
    ScaleRGB=true
   )
   &ScaleColor(
    Color Value To Scale=$Particles.Color
    Scale Alpha=&Multiply_Float(A=$User.Global Alpha Kill3, B=&FN_FrontEndFloatModifier(Gameplay=3, Front End=3.5))
    ScaleA=true
    ScaleRGB=true
   ) {module=ScaleColor001}
   &SolveForcesAndVelocity(Clamp Velocity=false, Limit Acceleration=false)
   &DynamicMaterialParameters(
    Index 0 Param 2=1
    Index 0 Param 3=1
    Index 0 Param 4=&FloatFromCurve(FloatCurve=DataInterface() {propertyValues {Curve {keys [<at 0 1>,<at 1 0.22>]}}})
    Module.Param0WriteEnabled=false
    Module.Param1WriteEnabled=true
    Module.Param2WriteEnabled=true
    Module.Param3WriteEnabled=true
   )
   &KillParticles(Kill Particles=&SetBoolByFloatComparison(A=$User.TransformProgress, B=0.9))
   &KillParticles(Kill Particles=$User.bisHandHidden) {module=KillParticles001}
  ]
  rendererClasses [&NiagaraSpriteRendererProperties]
  renderers [{rendererClass=&NiagaraSpriteRendererProperties}]
 }
 Hand_Smoke001 {
  emitterUpdateScript [
   &EmitterState(Life Cycle Mode=::NewEnumerator1, Loop Behavior=::NewEnumerator1, Loop Duration Mode=::NewEnumerator1, UseLoopDelay=false)
   &SpawnBurst_Instantaneous(Spawn Count=2, Use Spawn Probability=false, UseLoopCountLimit=false)
  ]
  particleSpawnScript [
   &FN_SetSpawnGroup(Number of Spawn Groups=2)
   &InitializeParticle(
    Color=<rgba 0.2604 5 0.5232 1>
    Mesh Scale=<0.175 0.175 0.055>
    Mesh Scale Mode=::NewEnumerator4
    UsePositionOffset=false
    Write Material Random=true
   )
   &CameraOffset(Camera Offset Amount=5)
   &DynamicMaterialParameters(
    Index 0 Param 1=&RandomRangeFloat(Minimum=0, Maximum=1)
    Index 0 Param 2=1
    Index 0 Param 3=5
    Index 0 Param 4=1
    Module.Param0WriteEnabled=true
    Module.Param0WriteEnabled1=true
    Module.Param0WriteEnabled2=true
    Module.Param0WriteEnabled3=true
    Module.Param1WriteEnabled=true
    Module.Param1WriteEnabled1=true
    Module.Param1WriteEnabled2=true
    Module.Param1WriteEnabled3=true
    Module.Param2WriteEnabled=true
    Module.Param2WriteEnabled1=true
    Module.Param2WriteEnabled2=true
    Module.Param2WriteEnabled3=true
    Module.Param3WriteEnabled=true
    Module.Param3WriteEnabled1=true
    Module.Param3WriteEnabled2=true
    Module.Param3WriteEnabled3=true
    Write Parameter Index 1=true
    Write Parameter Index 2=true
    Write Parameter Index 3=true
   )
   &InitialMeshOrientation(
    Rotation=&SelectVectorFromArray(Direct Array Index=$Particles.SpawnGroup, Vector Selection Array=DataInterface() {propertyValues {FloatData [<0 0 0.25>, <0 0 0.25>]}})
    Use Orientation Vector=true
    Use Rotation Vector=true
   ) {module=InitialMeshOrientation001}
  ]
  particleUpdateScript [
   &ParticleState(Kill Particles When Lifetime Has Elapsed=false, Let Infinitely Lived Particles Die When Emitter Deactivates=true)
   &SkeletalMeshLocation(
    Bone / Socket Index=$Particles.SpawnGroup
    Bone Sampling Mode=::NewEnumerator4
    MaskSpawnGroup=false
    Orientation Sampling=::NewEnumerator1
    Sampled Position Offset=&SelectVectorFromArray(Direct Array Index=$Particles.SpawnGroup, Vector Selection Array=DataInterface() {propertyValues {FloatData [<10 0 0>, <-10 0 0>]}})
    Skeletal Mesh=DataInterface() {propertyValues {
      DefaultMesh=null
      ExcludeBoneName=root
      OptionalComponentClass=null
      PreviewMesh=&M_MED_PiteousKicks
      SkinningMode=SkinOnTheFly
      SoftSourceActor=null
      SourceMode=Default
      UvSetIndex=0
      WholeMeshLOD=-1
      bExcludeBone=false
      bReadDeformedGeometry=true
      bRequireCurrentFrameData=true
      ComponentTags []
      FilteredBones [hand_l,hand_r]
      FilteredSockets []
      MeshUserParameter {parameter {name=None, typeDefHandle {registeredTypeIndex=1}, varData []}}
      SamplingRegions []
     }}
   )
   &InterpolateOverTime(Rate Of Change=10, Target Vector Value:NiagaraPosition=$Particles.Position, Variable Type=::NewEnumerator2)
   /SetVariables(Particles.Interpolate Dynamic=&MakeFloat4FromVectorAndFloat("Vector (RGB)"=$Output.InterpolateOverTime.Moving Average, "Float (Alpha)"=1)) {isSetParameters=true}
   &ScaleColor(ScaleA=true, ScaleRGB=true)
   &ScaleColor(
    Color Value To Scale=$Particles.Color
    Scale Alpha=&Multiply_Float(A=$User.Global Alpha Kill3, B=&FN_FrontEndFloatModifier(Gameplay=1, Front End=0.5))
    Scale RGB=&Multiply_VectorByFloat(
     Vector=&Multiply_VectorByFloat(
      Vector=&VectorFromFloat(Value=&FNTimeOfDayScalar(Night Scale=0.5, Morning Scale=0.25, Evening Scale=0.25))
      Float=&FN_FrontEndFloatModifier(Gameplay=1, Front End=1.5)
     )
     Float=&Multiply_Float(
      A=&FloatFromCurve(FloatCurve=DataInterface() {propertyValues {Curve {keys [<at 0 1>,<at 1 10>]}}}, CurveIndex=$User.Dissolve Edge Intensity)
      B=1
     )
    )
    ScaleA=true
    ScaleRGB=true
   ) {module=ScaleColor001}
   &KillParticles(Kill Particles=&SetBoolByFloatComparison(A=$User.TransformProgress, B=0.9))
   &KillParticles(Kill Particles=$User.bisHandHidden) {module=KillParticles001}
  ]
  rendererClasses [&NiagaraMeshRendererProperties]
  renderers [{rendererClass=&NiagaraMeshRendererProperties, rendererIndex=0}]
 }
 ParticleDecale_Eyes {
  emitterUpdateScript [
   &EmitterState(
    Inactive Response=::NewEnumerator1
    Life Cycle Mode=::NewEnumerator1
    Loop Behavior=::NewEnumerator1
    Loop Duration Mode=::NewEnumerator1
    UseLoopDelay=false
   )
   &SpawnBurst_Instantaneous(Spawn Count=2, Use Spawn Probability=false)
  ]
  particleSpawnScript [
   &FN_SetSpawnGroup(Number of Spawn Groups=2)
   &InitializeParticle(Color=<rgba 0 1 0.25 1>, Sprite Size Mode=::NewEnumerator3, Uniform Sprite Size=8, UsePositionOffset=false, Write Material Random=true)
   &CameraOffset(Camera Offset Amount=10)
  ]
  particleUpdateScript [
   &ParticleState(Kill Particles When Lifetime Has Elapsed=false, Let Infinitely Lived Particles Die When Emitter Deactivates=true)
   &SkeletalMeshLocation(
    Bone / Socket Index=1
    Bone Sampling Mode=::NewEnumerator4
    MaskSpawnGroup=false
    Sampled Position Offset=&SelectVectorFromArray(Direct Array Index=$Particles.SpawnGroup, Vector Selection Array=DataInterface() {propertyValues {FloatData [<4 10 -3>, <4 10 3>]}})
    Skeletal Mesh=DataInterface() {propertyValues {
      DefaultMesh=null
      ExcludeBoneName=root
      OptionalComponentClass=null
      PreviewMesh=&M_MED_PiteousKicks
      SkinningMode=SkinOnTheFly
      SoftSourceActor=null
      SourceMode=Default
      UvSetIndex=0
      WholeMeshLOD=-1
      bExcludeBone=false
      bReadDeformedGeometry=true
      bRequireCurrentFrameData=true
      ComponentTags []
      FilteredBones [head]
      FilteredSockets []
      MeshUserParameter {parameter {name=None, typeDefHandle {registeredTypeIndex=1}, varData []}}
      SamplingRegions []
     }}
   )
   &SolveForcesAndVelocity(Clamp Velocity=false, Limit Acceleration=false)
   &DynamicMaterialParameters(
    Index 0 Param 1=0.2
    Index 0 Param 3=1
    Index 0 Param 4=&RandomRangeFloat(Minimum=0, Maximum=1)
    Index 1 Param 1=20
    Module.Param0WriteEnabled=true
    Module.Param0WriteEnabled1=true
    Module.Param1WriteEnabled=true
    Module.Param1WriteEnabled1=false
    Module.Param2WriteEnabled=true
    Module.Param2WriteEnabled1=true
    Module.Param3WriteEnabled=true
    Module.Param3WriteEnabled1=true
    Write Parameter Index 1=true
   )
   &ScaleColor(
    Scale Alpha=&Multiply_Float(
     A=&Multiply_Float(A=$User.Global Alpha Kill2, B=$User.Dissolve Progress)
     B=&FN_FrontEndFloatModifier(Gameplay=1, Front End=1.3)
    )
    Scale RGB=&Multiply_VectorByFloat(
     Vector=&Multiply_VectorByFloat(
      Vector=&VectorFromFloat(Value=&FNTimeOfDayScalar(Night Scale=0.5, Morning Scale=0.25, Evening Scale=0.25))
      Float=&FN_FrontEndFloatModifier(Gameplay=1, Front End=1)
     )
     Float=1
    )
    ScaleA=true
    ScaleRGB=true
   )
   &KillParticles(Kill Particles=&SetBoolByFloatComparison(A=$User.TransformProgress, B=0.6))
   /SetVariables(Particles.VisibilityTag=&MakeCustomIntFromBool(Bool=$User.bisHeadHIdden, True INT=1, False INT=0)) {isSetParameters=true, module=SetVariables_#1}
  ]
  rendererClasses [&NiagaraSpriteRendererProperties]
  renderers [
   {rendererClass=&NiagaraSpriteRendererProperties}
   {rendererClass=&NiagaraSpriteRendererProperties, rendererIndex=1}
  ]
 }
 ParticleDecale_Legs {
  emitterUpdateScript [
   &EmitterState(
    Inactive Response=::NewEnumerator1
    Life Cycle Mode=::NewEnumerator1
    Loop Behavior=::NewEnumerator1
    Loop Duration Mode=::NewEnumerator1
    UseLoopDelay=false
   )
   &SpawnBurst_Instantaneous(Spawn Count=10, Use Spawn Probability=false)
  ]
  particleSpawnScript [
   &FN_SetSpawnGroup(Number of Spawn Groups=10)
   &InitializeParticle(
    Color=<rgba 0 1 0.25 1>
    Sprite Rotation Mode=::NewEnumerator2
    Sprite Size Mode=::NewEnumerator3
    Uniform Sprite Size=35
    UsePositionOffset=false
    Write Material Random=true
   )
   &CameraOffset(Camera Offset Amount=10)
  ]
  particleUpdateScript [
   &ParticleState(Kill Particles When Lifetime Has Elapsed=false, Let Infinitely Lived Particles Die When Emitter Deactivates=true)
   &SkeletalMeshLocation(
    Bone / Socket Index=$Particles.SpawnGroup
    Bone Sampling Mode=::NewEnumerator4
    MaskSpawnGroup=false
    Sampled Position Offset=&SelectVectorFromArray(
     Direct Array Index=$Particles.SpawnGroup
     Vector Selection Array=DataInterface() {propertyValues {FloatData [<0 0 0>, <0 0 0>, <10 0 0>, <-10 0 0>, <0 0 0>, <0 0 0>, <-5 0 0>, <-5 0 0>, <0 0 0>, <0 0 0>]}}
    )
    Skeletal Mesh=DataInterface() {propertyValues {
      DefaultMesh=null
      ExcludeBoneName=root
      OptionalComponentClass=null
      PreviewMesh=&M_MED_PiteousKicks
      SkinningMode=SkinOnTheFly
      SoftSourceActor=null
      SourceMode=Default
      UvSetIndex=0
      WholeMeshLOD=-1
      bExcludeBone=false
      bReadDeformedGeometry=true
      bRequireCurrentFrameData=true
      ComponentTags []
      FilteredBones [calf_l,calf_r,foot_l,foot_r,ball_l,ball_r,lowerarm_l,lowerarm_r,hand_l,hand_r]
      FilteredSockets []
      MeshUserParameter {parameter {name=None, typeDefHandle {registeredTypeIndex=1}, varData []}}
      SamplingRegions []
     }}
   )
   &SolveForcesAndVelocity(Clamp Velocity=false, Limit Acceleration=false)
   &DynamicMaterialParameters(
    Index 0 Param 1=0.7
    Index 0 Param 3=1
    Index 0 Param 4=&RandomRangeFloat(Minimum=0, Maximum=1)
    Index 1 Param 1=20
    Module.Param0WriteEnabled=true
    Module.Param0WriteEnabled1=true
    Module.Param1WriteEnabled=true
    Module.Param1WriteEnabled1=false
    Module.Param2WriteEnabled=true
    Module.Param2WriteEnabled1=true
    Module.Param3WriteEnabled=true
    Module.Param3WriteEnabled1=true
    Write Parameter Index 1=true
   )
   &ScaleColor(
    Scale Alpha=&Multiply_Float(
     A=&Multiply_Float(A=$User.Global Alpha Kill3, B=$User.Dissolve Progress)
     B=&FN_FrontEndFloatModifier(Gameplay=1, Front End=1)
    )
    Scale RGB=&Multiply_VectorByFloat(
     Vector=&Multiply_VectorByFloat(
      Vector=&VectorFromFloat(Value=&FNTimeOfDayScalar(Night Scale=0.5, Morning Scale=0.25, Evening Scale=0.25))
      Float=&FN_FrontEndFloatModifier(Gameplay=1, Front End=1)
     )
     Float=1.2
    )
    ScaleA=true
    ScaleRGB=true
   )
   &KillParticles(Kill Particles=&SetBoolByFloatComparison(A=$User.TransformProgress, B=0.9))
   &ScaleColor(
    Scale Alpha=&Multiply_Float(
     A=1
     B=&MakeCustomFloatFromBool(
      Bool=$User.bisFootHidden
      True Float=&SelectFloatFromArray(Direct Array Index=$Particles.SpawnGroup, Float Selection Array=DataInterface() {propertyValues {FloatData [0,0,0,0,0,0,1,1,1,1,1]}})
      False Float=1
     )
    )
    ScaleA=true
    ScaleRGB=false
   ) {module=ScaleColor001}
   /SetVariables(Particles.VisibilityTag=&MakeCustomIntFromBool(
     Bool=$User.bisHandHidden
     True INT=&MakeCustomIntFromBool(
      Bool=$User.bisFootHidden
      True INT=1
      False INT=&SelectIntFromArray(Direct Array Index=$Particles.SpawnGroup, Int Selection Array=DataInterface() {propertyValues {IntData [0,1,0,0,0,0,1,1,1,1]}})
     )
     False INT=&MakeCustomIntFromBool(
      Bool=$User.bisFootHidden
      True INT=&SelectIntFromArray(Direct Array Index=$Particles.SpawnGroup, Int Selection Array=DataInterface() {propertyValues {IntData [1,1,1,1,1,1,0,0,0,0]}})
      False INT=0
     )
    )) {isSetParameters=true}
  ]
  rendererClasses [&NiagaraSpriteRendererProperties]
  renderers [
   {rendererClass=&NiagaraSpriteRendererProperties}
   {rendererClass=&NiagaraSpriteRendererProperties, rendererIndex=1}
  ]
 }
 Trail_Arm_L {
  emitterUpdateScript [
   &EmitterState(Life Cycle Mode=::NewEnumerator1, Loop Behavior=::NewEnumerator1, Loop Duration Mode=::NewEnumerator1, UseLoopDelay=false)
   &SpawnBurst_Instantaneous(Spawn Count=20, Use Spawn Probability=false, UseLoopCountLimit=false)
  ]
  particleSpawnScript [
   &FN_SetSpawnGroup(Inclusivity=true, Normalize Spawn Group=true, Particles Per Group=20, Pattern=::NewEnumerator0, Set Ribbon ID and Ribbon Link Order=true)
   &InitializeParticle(Lifetime=5, Ribbon Width=1.5, Ribbon Width Mode=::NewEnumerator1, UsePositionOffset=false, Write Material Random=true)
  ]
  particleUpdateScript [
   &ParticleState(Kill Particles When Lifetime Has Elapsed=false, Let Infinitely Lived Particles Die When Emitter Deactivates=true)
   &FN_BoneFlow(
    Angle:NiagaraFloat=&Multiply_Float(A=$Particles.Normalized Spawn Group, B=-1.3)
    BoneID:NiagaraInt32=&ModuloInt(A=$Particles.SpawnGroup, B=2)
    Enable Position Offsets=true
    Height:NiagaraFloat=&Multiply_Float(A=$Particles.Normalized Spawn Group, B=1.5)
    Offset Start:Vector3f=&SelectVectorFromArray(
     Direct Array Index=&ModuloInt(A=$Particles.SpawnGroup, B=2)
     Vector Selection Array=DataInterface() {propertyValues {FloatData [<15 0 0>, <-15 0 0>]}}
    )
    Radius:NiagaraFloat=&FloatFromCurve(
     FloatCurve=DataInterface() {propertyValues {Curve {keys [
         {arriveTangent=27.7757, interpMode=RCIM_Cubic, leaveTangent=27.7757, tangentMode=RCTM_User, time=0, value=-5}
         {arriveTangent=21.9032, interpMode=RCIM_Cubic, leaveTangent=21.9032, tangentMode=RCTM_User, time=0.3, value=4}
         {arriveTangent=-8.6266, interpMode=RCIM_Cubic, leaveTangent=-8.6266, tangentMode=RCTM_User, time=0.7, value=5.5}
         {arriveTangent=0.5728, interpMode=RCIM_Cubic, leaveTangent=0.5728, tangentMode=RCTM_User, time=1, value=0}
        ]}}}
     CurveIndex=$Particles.Normalized Spawn Group
     Scale Curve=&SelectFloatFromArray(
      Direct Array Index=&ModuloInt(A=$Particles.SpawnGroup, B=2)
      Float Selection Array=DataInterface() {propertyValues {FloatData [1]}}
     )
    )
    SkeletalMesh=DataInterface() {propertyValues {
      DefaultMesh=null
      ExcludeBoneName=root
      OptionalComponentClass=null
      PreviewMesh=&M_MED_PiteousKicks
      SkinningMode=SkinOnTheFly
      SoftSourceActor=null
      SourceMode=Default
      UvSetIndex=0
      WholeMeshLOD=-1
      bExcludeBone=false
      bReadDeformedGeometry=true
      bRequireCurrentFrameData=true
      ComponentTags []
      FilteredBones [hand_l,lowerarm_l,upperarm_l]
      FilteredSockets []
      MeshUserParameter {parameter {name=None, typeDefHandle {registeredTypeIndex=1}, varData []}}
      SamplingRegions []
     }}
   )
   &ScaleRibbonWidth(Ribbon Width Scale=&FloatFromCurve(FloatCurve=DataInterface() {propertyValues {Curve {keys [<at 0 1.5>,<at 1 0.5>]}}}, CurveIndex=$Particles.RibbonLinkOrder, Scale Curve=1.5))
   /SetVariables(Particles.Ribbon B Width:NiagaraFloat=&Multiply_Float(A=$Particles.RibbonWidth, B=4.5)) {isSetParameters=true, module=SetVariables_#2}
   &ScaleColor(
    Scale Alpha=&FloatFromCurve(
     FloatCurve=DataInterface() {propertyValues {Curve {keys [
         {interpMode=RCIM_Cubic, tangentMode=RCTM_User, time=0.25, value=0}
         {arriveTangent=10.0679, interpMode=RCIM_Cubic, leaveTangent=-2.8986, tangentMode=RCTM_Break, time=0.42, value=1}
         {arriveTangent=-0.0273, interpMode=RCIM_Cubic, leaveTangent=-0.0273, tangentMode=RCTM_User, time=0.7, value=0}
        ]}}}
     CurveIndex=$Particles.RibbonLinkOrder
     Scale Curve=0.3
    )
    Scale RGB=&MakeVectorFromLinearColorRGB(Color=&ColorFromCurve(
      ColorCurve=DataInterface() {propertyValues {
        AlphaCurve {keys []}
        BlueCurve {keys [<at 0.2104 0.2996>,<at 0.543 0.1544 RCIM_Cubic>]}
        GreenCurve {keys [<at 0.2104 1>,<at 0.543 1 RCIM_Cubic>]}
        RedCurve {keys [<at 0.2104 0>,<at 0.543 0 RCIM_Cubic>]}
       }}
      CurveIndex=$StackContext.RibbonLinkOrder
     ))
    ScaleA=true
    ScaleRGB=true
   )
   &ScaleColor(
    Color Value To Scale=$Particles.Color
    Scale Alpha=&Multiply_Float(
     A=$User.Global Alpha Kill3
     B=&FN_FrontEndFloatModifier(
      Gameplay=1
      Front End=&FN_PickFloatDependOnPlatform(TargetPlatformActive=0.5, TargetPlatformNoActive=0.2)
     )
    )
    Scale RGB=&Multiply_VectorByFloat(
     Vector=&Multiply_VectorByFloat(
      Vector=&VectorFromFloat(Value=&FNTimeOfDayScalar(Night Scale=0.5, Morning Scale=0.25, Evening Scale=0.25))
      Float=&FN_FrontEndFloatModifier(Gameplay=1, Front End=1.5)
     )
     Float=&Multiply_Float(
      A=&FloatFromCurve(FloatCurve=DataInterface() {propertyValues {Curve {keys [<at 0 1>,<at 1 10>]}}}, CurveIndex=$User.Dissolve Edge Intensity)
      B=1
     )
    )
    ScaleA=true
    ScaleRGB=true
   ) {module=ScaleColor001}
   &DynamicMaterialParameters(
    Index 0 Param 2=1
    Index 0 Param 3=$User.Dissolve Progress
    Index 0 Param 4=$User.Dissolve Edge Intensity
    Module.Param0WriteEnabled=true
    Module.Param1WriteEnabled=true
    Module.Param2WriteEnabled=true
    Module.Param3WriteEnabled=true
   )
   &KillParticles(Kill Particles=&SetBoolByFloatComparison(A=$User.TransformProgress, B=0.9))
   &KillParticles(Kill Particles=$User.bisHandHidden) {module=KillParticles001}
  ]
  rendererClasses [&NiagaraRibbonRendererProperties]
  renderers [
   {rendererClass=&NiagaraRibbonRendererProperties}
   {rendererClass=&NiagaraRibbonRendererProperties, rendererIndex=1}
  ]
 }
 Trail_Arm_R {
  emitterUpdateScript [
   &EmitterState(Life Cycle Mode=::NewEnumerator1, Loop Behavior=::NewEnumerator1, Loop Duration Mode=::NewEnumerator1, UseLoopDelay=false)
   &SpawnBurst_Instantaneous(Spawn Count=20, Use Spawn Probability=false, UseLoopCountLimit=false)
  ]
  particleSpawnScript [
   &FN_SetSpawnGroup(Inclusivity=true, Normalize Spawn Group=true, Particles Per Group=20, Pattern=::NewEnumerator0, Set Ribbon ID and Ribbon Link Order=true)
   &InitializeParticle(Lifetime=5, Ribbon Width=1.5, Ribbon Width Mode=::NewEnumerator1, UsePositionOffset=false, Write Material Random=true)
  ]
  particleUpdateScript [
   &ParticleState(Kill Particles When Lifetime Has Elapsed=false, Let Infinitely Lived Particles Die When Emitter Deactivates=true)
   &FN_BoneFlow(
    Angle:NiagaraFloat=&Multiply_Float(A=$Particles.Normalized Spawn Group, B=-1.3)
    BoneID:NiagaraInt32=&ModuloInt(A=$Particles.SpawnGroup, B=2)
    Enable Position Offsets=true
    Height:NiagaraFloat=&Multiply_Float(A=$Particles.Normalized Spawn Group, B=1.5)
    Offset Start:Vector3f=&SelectVectorFromArray(
     Direct Array Index=&ModuloInt(A=$Particles.SpawnGroup, B=2)
     Vector Selection Array=DataInterface() {propertyValues {FloatData [<15 0 0>, <-15 0 0>]}}
    )
    Radius:NiagaraFloat=&FloatFromCurve(
     FloatCurve=DataInterface() {propertyValues {Curve {keys [
         {arriveTangent=27.7757, interpMode=RCIM_Cubic, leaveTangent=27.7757, tangentMode=RCTM_User, time=0, value=-5}
         {arriveTangent=21.9032, interpMode=RCIM_Cubic, leaveTangent=21.9032, tangentMode=RCTM_User, time=0.3, value=4}
         {arriveTangent=-8.6266, interpMode=RCIM_Cubic, leaveTangent=-8.6266, tangentMode=RCTM_User, time=0.7, value=5.5}
         {arriveTangent=0.5728, interpMode=RCIM_Cubic, leaveTangent=0.5728, tangentMode=RCTM_User, time=1, value=0}
        ]}}}
     CurveIndex=$Particles.Normalized Spawn Group
     Scale Curve=&SelectFloatFromArray(
      Direct Array Index=&ModuloInt(A=$Particles.SpawnGroup, B=2)
      Float Selection Array=DataInterface() {propertyValues {FloatData [-1]}}
     )
    )
    SkeletalMesh=DataInterface() {propertyValues {
      DefaultMesh=null
      ExcludeBoneName=root
      OptionalComponentClass=null
      PreviewMesh=&M_MED_PiteousKicks
      SkinningMode=SkinOnTheFly
      SoftSourceActor=null
      SourceMode=Default
      UvSetIndex=0
      WholeMeshLOD=-1
      bExcludeBone=false
      bReadDeformedGeometry=true
      bRequireCurrentFrameData=true
      ComponentTags []
      FilteredBones [hand_r,lowerarm_r,upperarm_r]
      FilteredSockets []
      MeshUserParameter {parameter {name=None, typeDefHandle {registeredTypeIndex=1}, varData []}}
      SamplingRegions []
     }}
   )
   &ScaleRibbonWidth(Ribbon Width Scale=&FloatFromCurve(FloatCurve=DataInterface() {propertyValues {Curve {keys [<at 0 1.5>,<at 1 0.5>]}}}, CurveIndex=$Particles.RibbonLinkOrder, Scale Curve=1.5))
   /SetVariables(Particles.Ribbon B Width:NiagaraFloat=&Multiply_Float(A=$Particles.RibbonWidth, B=4.5)) {isSetParameters=true, module=SetVariables_#2}
   &ScaleColor(
    Scale Alpha=&FloatFromCurve(
     FloatCurve=DataInterface() {propertyValues {Curve {keys [
         {interpMode=RCIM_Cubic, tangentMode=RCTM_User, time=0.25, value=0}
         {arriveTangent=10.0679, interpMode=RCIM_Cubic, leaveTangent=-2.8986, tangentMode=RCTM_Break, time=0.42, value=1}
         {arriveTangent=-0.0273, interpMode=RCIM_Cubic, leaveTangent=-0.0273, tangentMode=RCTM_User, time=0.7, value=0}
        ]}}}
     CurveIndex=$Particles.RibbonLinkOrder
     Scale Curve=0.3
    )
    Scale RGB=&MakeVectorFromLinearColorRGB(Color=&ColorFromCurve(
      ColorCurve=DataInterface() {propertyValues {
        AlphaCurve {keys []}
        BlueCurve {keys [<at 0.2104 0.2996>,<at 0.543 0.1544 RCIM_Cubic>]}
        GreenCurve {keys [<at 0.2104 1>,<at 0.543 1 RCIM_Cubic>]}
        RedCurve {keys [<at 0.2104 0>,<at 0.543 0 RCIM_Cubic>]}
       }}
      CurveIndex=$StackContext.RibbonLinkOrder
     ))
    ScaleA=true
    ScaleRGB=true
   )
   &ScaleColor(
    Color Value To Scale=$Particles.Color
    Scale Alpha=&Multiply_Float(
     A=$User.Global Alpha Kill1
     B=&FN_FrontEndFloatModifier(
      Gameplay=1
      Front End=&FN_PickFloatDependOnPlatform(TargetPlatformActive=0.5, TargetPlatformNoActive=0.2)
     )
    )
    Scale RGB=&Multiply_VectorByFloat(
     Vector=&Multiply_VectorByFloat(
      Vector=&VectorFromFloat(Value=&FNTimeOfDayScalar(Night Scale=0.5, Morning Scale=0.25, Evening Scale=0.25))
      Float=&FN_FrontEndFloatModifier(Gameplay=1, Front End=1.5)
     )
     Float=&Multiply_Float(
      A=&FloatFromCurve(FloatCurve=DataInterface() {propertyValues {Curve {keys [<at 0 1>,<at 1 10>]}}}, CurveIndex=$User.Dissolve Edge Intensity)
      B=1
     )
    )
    ScaleA=true
    ScaleRGB=true
   ) {module=ScaleColor001}
   &DynamicMaterialParameters(
    Index 0 Param 2=1
    Index 0 Param 3=$User.Dissolve Progress
    Index 0 Param 4=$User.Dissolve Edge Intensity
    Module.Param0WriteEnabled=true
    Module.Param1WriteEnabled=true
    Module.Param2WriteEnabled=true
    Module.Param3WriteEnabled=true
   )
   &KillParticles(Kill Particles=&SetBoolByFloatComparison(A=$User.TransformProgress, B=0.3))
   &KillParticles(Kill Particles=$User.bisHandHidden) {module=KillParticles001}
  ]
  rendererClasses [&NiagaraRibbonRendererProperties]
  renderers [
   {rendererClass=&NiagaraRibbonRendererProperties}
   {rendererClass=&NiagaraRibbonRendererProperties, rendererIndex=1}
  ]
 }
 Trail_Eyes {
  emitterUpdateScript [
   &EmitterState(Life Cycle Mode=::NewEnumerator1, Loop Behavior=::NewEnumerator1, Loop Duration Mode=::NewEnumerator1, UseLoopDelay=false)
   &SpawnBurst_Instantaneous(Spawn Count=2, Use Spawn Probability=false, UseLoopCountLimit=false)
  ]
  particleSpawnScript [
   &FN_SetSpawnGroup(Number of Spawn Groups=2)
   &InitializeParticle(
    Color=<rgba 0.0937 1 0.2074 2>
    Mesh Scale=&SelectVectorFromArray(Direct Array Index=$Particles.SpawnGroup, Vector Selection Array=DataInterface() {propertyValues {FloatData [<1 0.5 1>, <1 0.5 1.4>]}})
    Mesh Scale Mode=::NewEnumerator4
    UsePositionOffset=false
    Write Material Random=true
   )
   &InitialMeshOrientation(Mesh Axis To Orient=<0 0 1>, Mesh Orientation Mode=::NewEnumerator2, Use Rotation Vector=true)
   &CameraOffset(Camera Offset Amount=1.2)
  ]
  particleUpdateScript [
   &ParticleState(Kill Particles When Lifetime Has Elapsed=false, Let Infinitely Lived Particles Die When Emitter Deactivates=true)
   &SkeletalMeshLocation(
    Bone / Socket Index=4
    Bone Sampling Mode=::NewEnumerator4
    MaskSpawnGroup=false
    Sampled Position Offset=&SelectVectorFromArray(Direct Array Index=$Particles.SpawnGroup, Vector Selection Array=DataInterface() {propertyValues {FloatData [<4.3 9 2.8>, <4.3 9 -3>]}})
    Skeletal Mesh=$User.Skeletal Mesh
   )
   &InterpolateOverTime(
    Rate Of Change=8
    Target Vector Value:Vector3f=&ConvertPositionToVector(Input Position=$Particles.Position, Convert To Absolute Large World Vector=false)
    Variable Type=::NewEnumerator2
   )
   /SetVariables(Particles.Interpolate Dynamic=&MakeFloat4FromVectorAndFloat(
     "Vector (RGB)"=$StackContext.InterpolateOverTime.InterpolateOverTime_V3_.Moving Average
     "Float (Alpha)"=&Add_Float(A=$User.Interpolate Intensity, B=-0.5)
    )) {isSetParameters=true, module=SetVariables_#3}
   &ScaleMeshSize(Scale Factor=$Engine.Owner.Scale)
   &ScaleColor(
    Scale Alpha=&Multiply_Float(
     A=$User.Global Alpha Kill2
     B=&FN_FrontEndFloatModifier(
      Gameplay=&FN_PickFloatDependOnPlatform(TargetPlatformActive=0.5, TargetPlatformNoActive=0.2)
      Front End=&FN_PickFloatDependOnPlatform(TargetPlatformActive=0.6, TargetPlatformNoActive=0.3)
     )
    )
    Scale RGB=&Multiply_VectorByFloat(
     Vector=&Multiply_VectorByFloat(
      Vector=&VectorFromFloat(Value=&FNTimeOfDayScalar(Night Scale=0.5, Morning Scale=0.25, Evening Scale=0.25))
      Float=&FN_FrontEndFloatModifier(Gameplay=1, Front End=1.5)
     )
     Float=1
    )
    ScaleA=true
    ScaleRGB=true
   )
   &DynamicMaterialParameters(
    Index 0 Param 1=&RandomRangeFloat(Minimum=0, Maximum=1)
    Index 0 Param 2=&SelectFloatFromArray(Direct Array Index=$Particles.SpawnGroup, Float Selection Array=DataInterface() {propertyValues {FloatData [0.6,0.8]}})
    Index 0 Param 3=$User.Dissolve Progress
    Index 0 Param 4=$User.Dissolve Edge Intensity
    Module.Param0WriteEnabled=true
    Module.Param1WriteEnabled=true
    Module.Param2WriteEnabled=true
    Module.Param3WriteEnabled=true
   )
   &KillParticles(Kill Particles=&SetBoolByFloatComparison(A=$User.TransformProgress, B=0.6))
   /SetVariables(Particles.VisibilityTag=&MakeCustomIntFromBool(Bool=$User.bisHeadHIdden, True INT=1, False INT=0)) {isSetParameters=true, module=SetVariables_#1}
  ]
  rendererClasses [&NiagaraMeshRendererProperties]
  renderers [
   {rendererClass=&NiagaraMeshRendererProperties, rendererIndex=0}
   {rendererClass=&NiagaraMeshRendererProperties, rendererIndex=1}
  ]
 }
 Trail_Foot_L {
  emitterUpdateScript [
   &EmitterState(Life Cycle Mode=::NewEnumerator1, Loop Behavior=::NewEnumerator1, Loop Duration Mode=::NewEnumerator1, UseLoopDelay=false)
   &SpawnBurst_Instantaneous(Use Spawn Probability=false, UseLoopCountLimit=false)
  ]
  particleSpawnScript [
   &FN_SetSpawnGroup(Number of Spawn Groups=2)
   &InitializeParticle(Color=<rgba 0.0573 1 0.3044 0.8>, Mesh Scale=<2 2 3>, Mesh Scale Mode=::NewEnumerator4, UsePositionOffset=false, Write Material Random=true)
   &InitialMeshOrientation(Mesh Axis To Orient=<-1 0 0>, Mesh Orientation Mode=::NewEnumerator2, Rotation=<-0.25 0 0>, Use Rotation Vector=true)
   &CameraOffset(Camera Offset Amount=2)
  ]
  particleUpdateScript [
   &ParticleState(Kill Particles When Lifetime Has Elapsed=false, Let Infinitely Lived Particles Die When Emitter Deactivates=true)
   &SkeletalMeshLocation(
    Bone / Socket Index=2
    Bone Sampling Mode=::NewEnumerator4
    MaskSpawnGroup=false
    Sampled Position Offset=<-5 -10 -2>
    Skeletal Mesh=$User.Skeletal Mesh
   )
   &InterpolateOverTime(
    Rate Of Change=8
    Target Vector Value:Vector3f=&ConvertPositionToVector(Input Position=$Particles.Position, Convert To Absolute Large World Vector=false)
    Variable Type=::NewEnumerator2
   )
   /SetVariables(Particles.Interpolate Dynamic=&MakeFloat4FromVectorAndFloat(
     "Vector (RGB)"=$StackContext.InterpolateOverTime.InterpolateOverTime_V3_.Moving Average
     "Float (Alpha)"=&Add_Float(A=$User.Interpolate Intensity, B=0.5)
    )) {isSetParameters=true, module=SetVariables_#3}
   &ScaleMeshSize(Scale Factor=$Engine.Owner.Scale)
   &ScaleColor(
    Scale Alpha=&Multiply_Float(
     A=$User.Global Alpha Kill3
     B=&FN_FrontEndFloatModifier(
      Gameplay=&FN_PickFloatDependOnPlatform(TargetPlatformActive=1, TargetPlatformNoActive=0.5)
      Front End=&FN_PickFloatDependOnPlatform(TargetPlatformActive=0.5, TargetPlatformNoActive=0.3)
     )
    )
    Scale RGB=&Multiply_VectorByFloat(
     Vector=&Multiply_VectorByFloat(
      Vector=&VectorFromFloat(Value=&FNTimeOfDayScalar(Night Scale=0.5, Morning Scale=0.25, Evening Scale=0.25))
      Float=&FN_FrontEndFloatModifier(Gameplay=1, Front End=1.5)
     )
     Float=&Multiply_Float(
      A=&FloatFromCurve(FloatCurve=DataInterface() {propertyValues {Curve {keys [<at 0 1>,<at 1 10>]}}}, CurveIndex=$User.Dissolve Edge Intensity)
      B=&MakeCustomFloatFromBool(Bool=$User.bisFootHidden, True Float=0, False Float=1)
     )
    )
    ScaleA=true
    ScaleRGB=true
   )
   &DynamicMaterialParameters(
    Index 0 Param 1=&RandomRangeFloat(Minimum=0, Maximum=1)
    Index 0 Param 2=&SelectFloatFromArray(Direct Array Index=$Particles.SpawnGroup, Float Selection Array=DataInterface() {propertyValues {FloatData [1,0.4]}})
    Index 0 Param 3=$User.Dissolve Progress
    Index 0 Param 4=$User.Dissolve Edge Intensity
    Module.Param0WriteEnabled=true
    Module.Param1WriteEnabled=true
    Module.Param2WriteEnabled=true
    Module.Param3WriteEnabled=true
   )
   &KillParticles(Kill Particles=&SetBoolByFloatComparison(A=$User.TransformProgress, B=0.9))
  ]
  rendererClasses [&NiagaraMeshRendererProperties]
  renderers [
   {rendererClass=&NiagaraMeshRendererProperties, rendererIndex=0}
   {rendererClass=&NiagaraMeshRendererProperties, rendererIndex=1}
   {rendererClass=&NiagaraMeshRendererProperties, rendererIndex=2}
  ]
 }
 Trail_Foot_R {
  emitterUpdateScript [
   &EmitterState(Life Cycle Mode=::NewEnumerator1, Loop Behavior=::NewEnumerator1, Loop Duration Mode=::NewEnumerator1, UseLoopDelay=false)
   &SpawnBurst_Instantaneous(Use Spawn Probability=false, UseLoopCountLimit=false)
  ]
  particleSpawnScript [
   &FN_SetSpawnGroup(Number of Spawn Groups=2)
   &InitializeParticle(
    Color=<rgba 0.0573 1 0.3044 0.8>
    Mesh Scale=<2 2 2.5>
    Mesh Scale Mode=::NewEnumerator4
    UsePositionOffset=false
    Write Material Random=true
   )
   &InitialMeshOrientation(Mesh Orientation Mode=::NewEnumerator2, Rotation=<-0.25 0 0>, Use Rotation Vector=true)
   &CameraOffset(Camera Offset Amount=2)
  ]
  particleUpdateScript [
   &ParticleState(Kill Particles When Lifetime Has Elapsed=false, Let Infinitely Lived Particles Die When Emitter Deactivates=true)
   &SkeletalMeshLocation(
    Bone / Socket Index=3
    Bone Sampling Mode=::NewEnumerator4
    MaskSpawnGroup=false
    Sampled Position Offset=<4 10 0>
    Skeletal Mesh=$User.Skeletal Mesh
   )
   &InterpolateOverTime(
    Rate Of Change=8
    Target Vector Value:Vector3f=&ConvertPositionToVector(Input Position=$Particles.Position, Convert To Absolute Large World Vector=false)
    Variable Type=::NewEnumerator2
   )
   /SetVariables(Particles.Interpolate Dynamic=&MakeFloat4FromVectorAndFloat(
     "Vector (RGB)"=$StackContext.InterpolateOverTime.InterpolateOverTime_V3_.Moving Average
     "Float (Alpha)"=&Add_Float(A=$User.Interpolate Intensity, B=0.5)
    )) {isSetParameters=true, module=SetVariables_#3}
   &ScaleMeshSize(Scale Factor=$Engine.Owner.Scale)
   &ScaleColor(
    Scale Alpha=&Multiply_Float(
     A=$User.Global Alpha Kill3
     B=&FN_FrontEndFloatModifier(
      Gameplay=&FN_PickFloatDependOnPlatform(TargetPlatformActive=1, TargetPlatformNoActive=0.5)
      Front End=&FN_PickFloatDependOnPlatform(TargetPlatformActive=0.5, TargetPlatformNoActive=0.3)
     )
    )
    Scale RGB=&Multiply_VectorByFloat(
     Vector=&Multiply_VectorByFloat(
      Vector=&VectorFromFloat(Value=&FNTimeOfDayScalar(Night Scale=0.5, Morning Scale=0.25, Evening Scale=0.25))
      Float=&FN_FrontEndFloatModifier(Gameplay=1, Front End=1.5)
     )
     Float=&Multiply_Float(
      A=&FloatFromCurve(FloatCurve=DataInterface() {propertyValues {Curve {keys [<at 0 1>,<at 1 10>]}}}, CurveIndex=$User.Dissolve Edge Intensity)
      B=&MakeCustomFloatFromBool(Bool=$User.bisFootHidden, True Float=0, False Float=1)
     )
    )
    ScaleA=true
    ScaleRGB=true
   )
   &DynamicMaterialParameters(
    Index 0 Param 1=&RandomRangeFloat(Minimum=0, Maximum=1)
    Index 0 Param 2=&SelectFloatFromArray(Direct Array Index=$Particles.SpawnGroup, Float Selection Array=DataInterface() {propertyValues {FloatData [1,0.4]}})
    Index 0 Param 3=$User.Dissolve Progress
    Index 0 Param 4=$User.Dissolve Edge Intensity
    Module.Param0WriteEnabled=true
    Module.Param1WriteEnabled=true
    Module.Param2WriteEnabled=true
    Module.Param3WriteEnabled=true
   )
   &KillParticles(Kill Particles=&SetBoolByFloatComparison(A=$User.TransformProgress, B=0.9))
  ]
  rendererClasses [&NiagaraMeshRendererProperties]
  renderers [
   {rendererClass=&NiagaraMeshRendererProperties, rendererIndex=0}
   {rendererClass=&NiagaraMeshRendererProperties, rendererIndex=1}
   {rendererClass=&NiagaraMeshRendererProperties, rendererIndex=2}
  ]
 }
 Trail_Leg {
  emitterUpdateScript [
   &EmitterState(Life Cycle Mode=::NewEnumerator1, Loop Behavior=::NewEnumerator1, Loop Duration Mode=::NewEnumerator1, UseLoopDelay=false)
   &SpawnBurst_Instantaneous(Spawn Count=40, Use Spawn Probability=false, UseLoopCountLimit=false)
  ]
  particleSpawnScript [
   &FN_SetSpawnGroup(Inclusivity=true, Normalize Spawn Group=true, Particles Per Group=20, Pattern=::NewEnumerator0, Set Ribbon ID and Ribbon Link Order=true)
   &InitializeParticle(Lifetime=5, Ribbon Width=1.5, Ribbon Width Mode=::NewEnumerator1, UsePositionOffset=false, Write Material Random=true)
  ]
  particleUpdateScript [
   &ParticleState(Kill Particles When Lifetime Has Elapsed=false, Let Infinitely Lived Particles Die When Emitter Deactivates=true)
   &FN_BoneFlow(
    Angle:NiagaraFloat=&FloatFromCurve(
     FloatCurve=DataInterface() {propertyValues {Curve {keys [
         {arriveTangent=1.8289, interpMode=RCIM_Cubic, leaveTangent=1.8289, tangentMode=RCTM_User, time=0, value=0}
         {arriveTangent=1.7857, interpMode=RCIM_Cubic, leaveTangent=1.7857, time=0.3, value=0.65}
         {arriveTangent=1.5, interpMode=RCIM_Cubic, leaveTangent=1.5, time=0.7, value=1.25}
         {arriveTangent=1.2828, interpMode=RCIM_Cubic, leaveTangent=1.2828, tangentMode=RCTM_User, time=1, value=1.7}
        ]}}}
     CurveIndex=$Particles.Normalized Spawn Group
     Scale Curve=0.8
    )
    BoneID:NiagaraInt32=&ModuloInt(A=$Particles.SpawnGroup, B=2)
    Height:NiagaraFloat=&Multiply_Float(
     A=$Particles.Normalized Spawn Group
     B=&SelectFloatFromArray(Direct Array Index=$Particles.SpawnGroup, Float Selection Array=DataInterface() {propertyValues {FloatData [1.4,1]}})
    )
    Radius:NiagaraFloat=&FloatFromCurve(
     FloatCurve=DataInterface() {propertyValues {Curve {keys [
         {arriveTangent=-0.6799, interpMode=RCIM_Cubic, leaveTangent=-0.6799, tangentMode=RCTM_User, time=0, value=7}
         {arriveTangent=-4.2857, interpMode=RCIM_Cubic, leaveTangent=-4.2857, time=0.3571, value=5.9951}
         {arriveTangent=-8.7317, interpMode=RCIM_Cubic, leaveTangent=-8.7317, tangentMode=RCTM_User, time=0.7, value=4}
         {arriveTangent=-3.5832, interpMode=RCIM_Cubic, leaveTangent=-3.5832, tangentMode=RCTM_User, time=1, value=1}
        ]}}}
     CurveIndex=$Particles.Normalized Spawn Group
     Scale Curve=0.8
    )
    SideVector1:Vector3f=<1 1 1>
    SideVector2:Vector3f=<1 1 1>
    SideVector3:Vector3f=<1 1 1>
    SkeletalMesh=DataInterface() {propertyValues {
      DefaultMesh=null
      ExcludeBoneName=root
      OptionalComponentClass=null
      PreviewMesh=&M_MED_PiteousKicks
      SkinningMode=SkinOnTheFly
      SoftSourceActor=null
      SourceMode=Default
      UvSetIndex=0
      WholeMeshLOD=-1
      bExcludeBone=false
      bReadDeformedGeometry=true
      bRequireCurrentFrameData=true
      ComponentTags []
      FilteredBones [foot_l,calf_l,thigh_l,foot_r,calf_r,thigh_r]
      FilteredSockets []
      MeshUserParameter {parameter {name=None, typeDefHandle {registeredTypeIndex=1}, varData []}}
      SamplingRegions []
     }}
    Stabalize Twisting=true
   )
   &ScaleRibbonWidth(Ribbon Width Scale=&FloatFromCurve(
     FloatCurve=DataInterface() {propertyValues {Curve {keys [
         <at 0 0.5>
         {arriveTangent=1.8056, interpMode=RCIM_Cubic, leaveTangent=1.8056, tangentMode=RCTM_User, time=0.2, value=0.9}
         {arriveTangent=0.6908, tangentMode=RCTM_User, time=1, value=2}
        ]}}}
     CurveIndex=$Particles.RibbonLinkOrder
    ))
   /SetVariables(Particles.Ribbon B Width:NiagaraFloat=&Multiply_Float(A=$Particles.RibbonWidth, B=8)) {isSetParameters=true, module=SetVariables_#2}
   &ScaleColor(
    Scale Alpha=&FloatFromCurve(
     FloatCurve=DataInterface() {propertyValues {Curve {keys [
         {arriveTangent=1.4743, interpMode=RCIM_Cubic, leaveTangent=1.4743, tangentMode=RCTM_User, time=0, value=0}
         {arriveTangent=4.159, interpMode=RCIM_Cubic, leaveTangent=-3.8038, tangentMode=RCTM_Break, time=0.3, value=1}
         {arriveTangent=-0.2647, interpMode=RCIM_Cubic, leaveTangent=-0.2647, tangentMode=RCTM_User, time=0.9, value=0}
        ]}}}
     CurveIndex=$Particles.RibbonLinkOrder
     Scale Curve=0.5
    )
    Scale RGB=&MakeVectorFromLinearColorRGB(Color=&ColorFromCurve(
      ColorCurve=DataInterface() {propertyValues {
        AlphaCurve {keys []}
        BlueCurve {keys [<at 0.1537 0.2534 RCIM_Cubic>]}
        GreenCurve {keys [<at 0.1537 1 RCIM_Cubic>]}
        RedCurve {keys [<at 0.1537 0.042 RCIM_Cubic>]}
       }}
      CurveIndex=$StackContext.RibbonLinkOrder
     ))
    ScaleA=true
    ScaleRGB=true
   )
   &ScaleColor(
    Color Value To Scale=$Particles.Color
    Scale Alpha=&Multiply_Float(
     A=&Multiply_Float(
      A=$User.Global Alpha Kill3
      B=&FN_FrontEndFloatModifier(
       Gameplay=&FN_PickFloatDependOnPlatform(TargetPlatformActive=1, TargetPlatformNoActive=0.5)
       Front End=&FN_PickFloatDependOnPlatform(TargetPlatformActive=0.8, TargetPlatformNoActive=0.1)
      )
     )
     B=&MakeCustomFloatFromBool(Bool=$User.bisFootHidden, True Float=0, False Float=1)
    )
    Scale RGB=&Multiply_VectorByFloat(
     Vector=&Multiply_VectorByFloat(
      Vector=&VectorFromFloat(Value=&FNTimeOfDayScalar(Night Scale=0.5, Morning Scale=0.25, Evening Scale=0.25))
      Float=&FN_FrontEndFloatModifier(Gameplay=1, Front End=1.5)
     )
     Float=&Multiply_Float(
      A=&FloatFromCurve(FloatCurve=DataInterface() {propertyValues {Curve {keys [<at 0 1>,<at 1 10>]}}}, CurveIndex=$User.Dissolve Edge Intensity)
      B=1
     )
    )
    ScaleA=true
    ScaleRGB=true
   ) {module=ScaleColor001}
   &DynamicMaterialParameters(
    Index 0 Param 2=1
    Index 0 Param 3=$User.Dissolve Progress
    Index 0 Param 4=$User.Dissolve Edge Intensity
    Module.Param0WriteEnabled=true
    Module.Param1WriteEnabled=true
    Module.Param2WriteEnabled=true
    Module.Param3WriteEnabled=true
   )
   &KillParticles(Kill Particles=&SetBoolByFloatComparison(A=$User.TransformProgress, B=0.9))
  ]
  rendererClasses [&NiagaraRibbonRendererProperties]
  renderers [
   {rendererClass=&NiagaraRibbonRendererProperties}
   {rendererClass=&NiagaraRibbonRendererProperties, rendererIndex=1}
  ]
 }
 Trail_Shoulder_L {
  emitterUpdateScript [
   &EmitterState(Life Cycle Mode=::NewEnumerator1, Loop Behavior=::NewEnumerator1, Loop Duration Mode=::NewEnumerator1, UseLoopDelay=false)
   &SpawnBurst_Instantaneous(Use Spawn Probability=false, UseLoopCountLimit=false)
  ]
  particleSpawnScript [
   &InitializeParticle(
    Color=<rgba 0.1475 1 0.2192 1>
    Mesh Scale=<2.5 2.5 2.5>
    Mesh Scale Mode=::NewEnumerator4
    UsePositionOffset=false
    Write Material Random=true
   )
   &InitialMeshOrientation(Mesh Orientation Mode=::NewEnumerator2, Rotation=<0 0.025 0>, Use Rotation Vector=true)
   &CameraOffset(Camera Offset Amount=4)
  ]
  particleUpdateScript [
   &ParticleState(Kill Particles When Lifetime Has Elapsed=false, Let Infinitely Lived Particles Die When Emitter Deactivates=true)
   &SkeletalMeshLocation(Bone Sampling Mode=::NewEnumerator4, MaskSpawnGroup=false, Sampled Position Offset=<14 5 5>, Skeletal Mesh=$User.Skeletal Mesh)
   &InterpolateOverTime(
    Rate Of Change=8
    Target Vector Value:Vector3f=&ConvertPositionToVector(Input Position=$Particles.Position, Convert To Absolute Large World Vector=false)
    Variable Type=::NewEnumerator2
   )
   /SetVariables(Particles.Interpolate Dynamic=&MakeFloat4FromVectorAndFloat("Vector (RGB)"=$StackContext.InterpolateOverTime.InterpolateOverTime_V3_.Moving Average, "Float (Alpha)"=$User.Interpolate Intensity)) {isSetParameters=true, module=SetVariables_#3}
   &ScaleMeshSize(Scale Factor=$Engine.Owner.Scale)
   &ScaleColor(
    Scale Alpha=&Multiply_Float(
     A=$User.Global Alpha Kill2
     B=&FN_FrontEndFloatModifier(
      Gameplay=&FN_PickFloatDependOnPlatform(TargetPlatformActive=0.6, TargetPlatformNoActive=0.3)
      Front End=&FN_PickFloatDependOnPlatform(TargetPlatformActive=0.6, TargetPlatformNoActive=0.3)
     )
    )
    Scale RGB=&Multiply_VectorByFloat(
     Vector=&Multiply_VectorByFloat(
      Vector=&VectorFromFloat(Value=&FNTimeOfDayScalar(Night Scale=0.5, Morning Scale=0.25, Evening Scale=0.25))
      Float=&FN_FrontEndFloatModifier(Gameplay=1, Front End=1.5)
     )
     Float=&Multiply_Float(
      A=&FloatFromCurve(FloatCurve=DataInterface() {propertyValues {Curve {keys [<at 0 1>,<at 1 10>]}}}, CurveIndex=$User.Dissolve Edge Intensity)
      B=1
     )
    )
    ScaleA=true
    ScaleRGB=true
   )
   &DynamicMaterialParameters(
    Index 0 Param 1=&RandomRangeFloat(Minimum=0, Maximum=1)
    Index 0 Param 2=1
    Index 0 Param 3=$User.Dissolve Progress
    Index 0 Param 4=$User.Dissolve Edge Intensity
    Module.Param0WriteEnabled=true
    Module.Param1WriteEnabled=true
    Module.Param2WriteEnabled=true
    Module.Param3WriteEnabled=true
   )
   &KillParticles(Kill Particles=&SetBoolByFloatComparison(A=$User.TransformProgress, B=0.6))
  ]
  rendererClasses [&NiagaraMeshRendererProperties]
  renderers [
   {rendererClass=&NiagaraMeshRendererProperties, rendererIndex=0}
   {rendererClass=&NiagaraMeshRendererProperties, rendererIndex=1}
   {rendererClass=&NiagaraMeshRendererProperties, rendererIndex=2}
   {rendererClass=&NiagaraMeshRendererProperties, rendererIndex=3}
  ]
 }
 Trail_Shoulder_R {
  emitterUpdateScript [
   &EmitterState(Life Cycle Mode=::NewEnumerator1, Loop Behavior=::NewEnumerator1, Loop Duration Mode=::NewEnumerator1, UseLoopDelay=false)
   &SpawnBurst_Instantaneous(Spawn Count=2, Use Spawn Probability=false, UseLoopCountLimit=false)
  ]
  particleSpawnScript [
   &FN_SetSpawnGroup(Number of Spawn Groups=2)
   &InitializeParticle(
    Color=<rgba 0.1914 1 0.3019 1.3>
    Mesh Scale=&SelectVectorFromArray(Direct Array Index=$Particles.SpawnGroup, Vector Selection Array=DataInterface() {propertyValues {FloatData [<3 3.5 2.5>, <3.5 3 3>]}})
    Mesh Scale Mode=::NewEnumerator4
    UsePositionOffset=false
    Write Material Random=true
   )
   &InitialMeshOrientation(
    Mesh Orientation Mode=::NewEnumerator2
    Rotation=&SelectVectorFromArray(
     Direct Array Index=$Particles.SpawnGroup
     Vector Selection Array=DataInterface() {propertyValues {FloatData [<0.5 -0.1 -0.1>, <0.5 -0.1 -0.15>]}}
    )
    Use Rotation Vector=true
   )
   &CameraOffset(Camera Offset Amount=2)
  ]
  particleUpdateScript [
   &ParticleState(Kill Particles When Lifetime Has Elapsed=false, Let Infinitely Lived Particles Die When Emitter Deactivates=true)
   &SkeletalMeshLocation(
    Bone / Socket Index=1
    Bone Sampling Mode=::NewEnumerator4
    MaskSpawnGroup=false
    Sampled Position Offset=&SelectVectorFromArray(Direct Array Index=$Particles.SpawnGroup, Vector Selection Array=DataInterface() {propertyValues {FloatData [<-17 -2 11>, <-12 -3 7>]}})
    Skeletal Mesh=$User.Skeletal Mesh
   )
   &InterpolateOverTime(
    Rate Of Change=8
    Target Vector Value:Vector3f=&ConvertPositionToVector(Input Position=$Particles.Position, Convert To Absolute Large World Vector=false)
    Variable Type=::NewEnumerator2
   )
   /SetVariables(Particles.Interpolate Dynamic=&MakeFloat4FromVectorAndFloat("Vector (RGB)"=$StackContext.InterpolateOverTime.InterpolateOverTime_V3_.Moving Average, "Float (Alpha)"=$User.Interpolate Intensity)) {isSetParameters=true, module=SetVariables_#3}
   &ScaleMeshSize(Scale Factor=$Engine.Owner.Scale)
   &ScaleColor(
    Scale Alpha=&Multiply_Float(
     A=&Multiply_Float(A=$User.Global Alpha Kill2, B=0.6)
     B=&FN_FrontEndFloatModifier(
      Gameplay=&ClampFloat(Float=&FN_LOD_Distance_Remap(Start Distance=200, End Distance=300, Minimum Output=0, Maximum Output=1))
      Front End=0.9
     )
    )
    Scale RGB=&Multiply_VectorByFloat(
     Vector=&Multiply_VectorByFloat(
      Vector=&VectorFromFloat(Value=&FNTimeOfDayScalar(Night Scale=0.5, Morning Scale=0.25, Evening Scale=0.25))
      Float=&FN_FrontEndFloatModifier(Gameplay=1, Front End=1.5)
     )
     Float=&Multiply_Float(
      A=&FloatFromCurve(FloatCurve=DataInterface() {propertyValues {Curve {keys [<at 0 1>,<at 1 10>]}}}, CurveIndex=$User.Dissolve Edge Intensity)
      B=1
     )
    )
    ScaleA=true
    ScaleRGB=true
   )
   &DynamicMaterialParameters(
    Index 0 Param 1=&RandomRangeFloat(Minimum=0, Maximum=1)
    Index 0 Param 2=&SelectFloatFromArray(Direct Array Index=$Particles.SpawnGroup, Float Selection Array=DataInterface() {propertyValues {FloatData [1.4,1]}})
    Index 0 Param 3=$User.Dissolve Progress
    Index 0 Param 4=$User.Dissolve Edge Intensity
    Module.Param0WriteEnabled=true
    Module.Param1WriteEnabled=true
    Module.Param2WriteEnabled=true
    Module.Param3WriteEnabled=true
   )
   &KillParticles(Kill Particles=&SetBoolByFloatComparison(A=$User.TransformProgress, B=0.6))
  ]
  rendererClasses [&NiagaraMeshRendererProperties]
  renderers [
   {rendererClass=&NiagaraMeshRendererProperties, rendererIndex=0}
   {rendererClass=&NiagaraMeshRendererProperties, rendererIndex=1}
  ]
 }
 Trail_Shoulder_R02 {
  emitterUpdateScript [
   &EmitterState(Life Cycle Mode=::NewEnumerator1, Loop Behavior=::NewEnumerator1, Loop Duration Mode=::NewEnumerator1, UseLoopDelay=false)
   &SpawnBurst_Instantaneous(Spawn Count=2, Use Spawn Probability=false, UseLoopCountLimit=false)
  ]
  particleSpawnScript [
   &FN_SetSpawnGroup(Number of Spawn Groups=2)
   &InitializeParticle(Color=<rgba 0.099 1 0.2506 1.5>, Mesh Scale=<1 1 3>, Mesh Scale Mode=::NewEnumerator4, UsePositionOffset=false, Write Material Random=true)
   &InitialMeshOrientation(Mesh Orientation Mode=::NewEnumerator2, Rotation=<0.5 0 0>, Use Rotation Vector=true)
   &CameraOffset(Camera Offset Amount=3)
  ]
  particleUpdateScript [
   &ParticleState(Kill Particles When Lifetime Has Elapsed=false, Let Infinitely Lived Particles Die When Emitter Deactivates=true)
   &SkeletalMeshLocation(
    Bone / Socket Index=1
    Bone Sampling Mode=::NewEnumerator4
    MaskSpawnGroup=false
    Sampled Position Offset=<-12 5 -6>
    Skeletal Mesh=$User.Skeletal Mesh
   )
   &InterpolateOverTime(
    Rate Of Change=8
    Target Vector Value:Vector3f=&ConvertPositionToVector(Input Position=$Particles.Position, Convert To Absolute Large World Vector=false)
    Variable Type=::NewEnumerator2
   )
   /SetVariables(Particles.Interpolate Dynamic=&MakeFloat4FromVectorAndFloat("Vector (RGB)"=$StackContext.InterpolateOverTime.InterpolateOverTime_V3_.Moving Average, "Float (Alpha)"=$User.Interpolate Intensity)) {isSetParameters=true, module=SetVariables_#3}
   &ScaleMeshSize(Scale Factor=$Engine.Owner.Scale)
   &ScaleColor(
    Scale Alpha=&Multiply_Float(
     A=&Multiply_Float(
      A=$User.Global Alpha Kill2
      B=&FN_PickFloatDependOnPlatform(TargetPlatformActive=1, TargetPlatformNoActive=0.5)
     )
     B=&FN_FrontEndFloatModifier(
      Gameplay=&ClampFloat(Float=&FN_LOD_Distance_Remap(Start Distance=200, End Distance=300, Minimum Output=0, Maximum Output=1))
      Front End=1
     )
    )
    Scale RGB=&Multiply_VectorByFloat(
     Vector=&Multiply_VectorByFloat(
      Vector=&VectorFromFloat(Value=&FNTimeOfDayScalar(Night Scale=0.5, Morning Scale=0.25, Evening Scale=0.25))
      Float=&FN_FrontEndFloatModifier(Gameplay=1, Front End=1)
     )
     Float=&Multiply_Float(
      A=&FloatFromCurve(FloatCurve=DataInterface() {propertyValues {Curve {keys [<at 0 1>,<at 1 10>]}}}, CurveIndex=$User.Dissolve Edge Intensity)
      B=1
     )
    )
    ScaleA=true
    ScaleRGB=true
   )
   &DynamicMaterialParameters(
    Index 0 Param 1=&RandomRangeFloat(Minimum=0, Maximum=1)
    Index 0 Param 2=&SelectFloatFromArray(Direct Array Index=$Particles.SpawnGroup, Float Selection Array=DataInterface() {propertyValues {FloatData [0.8,0.2]}})
    Index 0 Param 3=$User.Dissolve Progress
    Index 0 Param 4=$User.Dissolve Edge Intensity
    Module.Param0WriteEnabled=true
    Module.Param1WriteEnabled=true
    Module.Param2WriteEnabled=true
    Module.Param3WriteEnabled=true
   )
   &KillParticles(Kill Particles=&SetBoolByFloatComparison(A=$User.TransformProgress, B=0.6))
  ]
  rendererClasses [&NiagaraMeshRendererProperties]
  renderers [
   {rendererClass=&NiagaraMeshRendererProperties, rendererIndex=0}
   {rendererClass=&NiagaraMeshRendererProperties, rendererIndex=1}
   {rendererClass=&NiagaraMeshRendererProperties, rendererIndex=2}
  ]
 }
}
summary {
 systemName=NS_PiteousKicks_Pine_Skin_Trail
 emitters [
  {emitterName=Trail_Leg, rendererClasses [&NiagaraRibbonRendererProperties]}
  {emitterName=Trail_Foot_L, rendererClasses [&NiagaraMeshRendererProperties]}
  {emitterName=Trail_Foot_R, rendererClasses [&NiagaraMeshRendererProperties]}
  {emitterName=Trail_Shoulder_L, rendererClasses [&NiagaraMeshRendererProperties]}
  {emitterName=Trail_Shoulder_R, rendererClasses [&NiagaraMeshRendererProperties]}
  {emitterName=Trail_Shoulder_R02, rendererClasses [&NiagaraMeshRendererProperties]}
  {emitterName=Trail_Eyes, rendererClasses [&NiagaraMeshRendererProperties]}
  {emitterName=ParticleDecale_Eyes, rendererClasses [&NiagaraSpriteRendererProperties]}
  {emitterName=ParticleDecale_Legs, rendererClasses [&NiagaraSpriteRendererProperties]}
  {emitterName=Trail_Arm_R, rendererClasses [&NiagaraRibbonRendererProperties]}
  {emitterName=Hand_Smoke, rendererClasses [&NiagaraSpriteRendererProperties]}
  {emitterName=Hand_Smoke001, rendererClasses [&NiagaraMeshRendererProperties]}
  {emitterName=Trail_Arm_L, rendererClasses [&NiagaraRibbonRendererProperties]}
  {bEnabled=false, emitterName=FN_Debug, rendererClasses [&NiagaraSpriteRendererProperties]}
 ]
 userVariables [
  User.bisFootHidden {type=NiagaraBool}
  User.bIsFrontEnd {type=NiagaraBool}
  User.bIsFrontEndPreview {type=NiagaraBool}
  User.bisHandHidden {type=NiagaraBool}
  User.bisHeadHIdden {type=NiagaraBool}
  User.Dissolve Edge Intensity {type=NiagaraFloat, defaultValue=0, description=""}
  User.Dissolve Progress {type=NiagaraFloat, defaultValue=1, description=""}
  User.Global Alpha {type=NiagaraFloat, defaultValue=1, description=""}
  User.Global Alpha Kill1 {type=NiagaraFloat, defaultValue=1, description=""}
  User.Global Alpha Kill2 {type=NiagaraFloat, defaultValue=1, description=""}
  User.Global Alpha Kill3 {type=NiagaraFloat, defaultValue=1, description=""}
  User.Interpolate Intensity {type=NiagaraFloat, defaultValue=0.7, description=""}
  User.Skeletal Mesh {
   type=NiagaraDataInterfaceSkeletalMesh
   defaultValue {
    struct=NiagaraExt_VariableValue_DataInterface
    value {dataInterface=&self:NiagaraDataInterfaceSkeletalMesh_0}
   }
  }
  User.TransformProgress {type=NiagaraFloat, defaultValue=0.5, description=""}
 ]
}
